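# Niagara system for the idle smoke effect on the VirtualShadowMale1H pickaxe cosmetic, built from two mesh-rendered emitters.
# SmokeIdle bursts mesh particles that persist past their lifetime until the emitter deactivates.
# SmokeParticles continuously spawns short-lived dark grey meshes at a vertical offset, drifting along +X and +Z while scaling over life and fading out.
# Curves and random values feed its dynamic material parameters to animate the smoke material per particle.
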NiagaraSystem /BRCosmetics/Weapons/FORT_Melee/Pickaxe_VirtualShadowMale1H/FX/NS_Pickaxe_VirtualShadowMale1H_Idle
dependencies {usedDataInterfaces [&NiagaraDataInterfaceCurve], usedDynamicInputs [], usedModules [&SystemState]}
emitters {
 SmokeIdle {
  emitterUpdateScript [
   &EmitterState(Life Cycle Mode=::NewEnumerator1, Loop Behavior=::NewEnumerator1, Loop Duration Mode=::NewEnumerator1, UseLoopDelay=false)
   &SpawnBurst_Instantaneous(Use Spawn Probability=false, UseLoopCountLimit=false)
  ]
  particleSpawnScript [&InitializeParticle(UsePositionOffset=false, Write Material Random=true)]
  particleUpdateScript [&ParticleState(Kill Particles When Lifetime Has Elapsed=false, Let Infinitely Lived Particles Die When Emitter Deactivates=true)]
  rendererClasses [&NiagaraMeshRendererProperties]
  renderers [{rendererClass=&NiagaraMeshRendererProperties, rendererIndex=0}]
 }
 SmokeParticles {
  emitterUpdateScript [
   &EmitterState(Life Cycle Mode=::NewEnumerator1, Loop Behavior=::NewEnumerator1, Loop Duration Mode=::NewEnumerator1, UseLoopDelay=false)
   &SpawnRate(Spawn Probability=0.5, Use Spawn Probability=false)
  ]
  particleSpawnScript [
   &InitializeParticle(
    Color=&RandomRangeLinearColor(Minimum=<rgba 0.1146 0.1146 0.1146 1>, Maximum=<rgba 0.0625 0.0625 0.0625 1>)
    Lifetime=&RandomRangeFloat(Minimum=1, Maximum=1.5)
    Mesh Scale=&Multiply_VectorByFloat(Vector=&RandomRangeVector(Minimum=<0.05 0.004 1>, Maximum=<0.08 0.005 1>), Float=3.9689)
    Mesh Scale Mode=::NewEnumerator4
    Position Offset=<0 0 50>
    Sprite Size Max=<0.08 0.005>
    Sprite Size Min=<0.05 0.004>
    Sprite Size Mode=::NewEnumerator2
    UsePositionOffset=true
    Write Material Random=true
   )
   &AddVelocity(Velocity=&RandomRangeVector(Minimum=<8 0 3>, Maximum=<13 0 10>))
  ]
  particleUpdateScript [
   &ParticleState()
   &DynamicMaterialParameters(
    Index 0 Param 1=&FloatFromCurve(FloatCurve=DataInterface() {propertyValues {Curve {keys [<at 0 0.5>,<at 1 0>]}}})
    Index 0 Param 2=&RandomRangeFloat(Minimum=20, Maximum=50)
    Index 0 Param 3=&FloatFromCurve(FloatCurve=DataInterface() {propertyValues {Curve {keys [<at 0 3>,<at 1 6>]}}})
    Index 0 Param 4=&FloatFromCurve(FloatCurve=DataInterface() {propertyValues {Curve {keys [<at 0 0.25>,<at 0.8 0.2>,<at 0.9 0.1>,<at 1 0>]}}})
    Module.Index0Float0Write=true
    Module.Index0Float1Write=true
    Module.Index0Float2Write=true
    Module.Index0Float3Write=true
   )
   &ScaleColor(
    Scale Alpha=&FloatFromCurve(FloatCurve=DataInterface() {propertyValues {Curve {keys [
         {interpMode=RCIM_Cubic, tangentMode=RCTM_User, time=0, value=0}
         {arriveTangent=2.5, interpMode=RCIM_Cubic, leaveTangent=2.5, time=0.2, value=1}
         {arriveTangent=-0.75, interpMode=RCIM_Cubic, leaveTangent=-0.75, time=0.4, value=1}
         {arriveTangent=-2.6667, interpMode=RCIM_Cubic, leaveTangent=-2.6667, time=0.6, value=0.7}
         {arriveTangent=-2.3293, interpMode=RCIM_Cubic, leaveTangent=-2.3293, time=0.7, value=0.2}
         {arriveTangent=-0.6667, interpMode=RCIM_Cubic, leaveTangent=-0.6667, time=0.8828, value=0.0412}
         {interpMode=RCIM_Cubic, tangentMode=RCTM_User, time=1, value=0}
        ]}}})
    ScaleA=true
    ScaleRGB=true
   )
   &ScaleMeshSize(Scale Factor=&VectorFromFloat(Value=&FloatFromCurve(FloatCurve=DataInterface() {propertyValues {Curve {keys [<at 0.1 1>,<at 10 8>]}}})))
   &SolveForcesAndVelocity(Clamp Velocity=false, Limit Acceleration=false)
  ]
  rendererClasses [&NiagaraMeshRendererProperties]
  renderers [{rendererClass=&NiagaraMeshRendererProperties, rendererIndex=0}]
 }
}
summary {
 systemName=NS_Pickaxe_VirtualShadowMale1H_Idle
 emitters [
  {emitterName=SmokeIdle, rendererClasses [&NiagaraMeshRendererProperties]}
  {emitterName=SmokeParticles, rendererClasses [&NiagaraMeshRendererProperties]}
 ]
 userVariables []
}
