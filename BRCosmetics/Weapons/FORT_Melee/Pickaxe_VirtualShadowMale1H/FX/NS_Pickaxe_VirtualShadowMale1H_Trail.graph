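# Niagara swing trail effect for the VirtualShadowMale1H pickaxe cosmetic, built from two ribbon emitters with dark coloring.
# FN_AnimTrail uses shared anim-trail modules to build a short-lived black ribbon between bone or socket indices, drawn by three renderers with a fading alpha.
# SmokeTrail spawns ribbon particles from a socket reader, widening and fading them through curves and dynamic material parameters.
# Smoke alpha also scales with time of day and dims in the front end.
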NiagaraSystem /BRCosmetics/Weapons/FORT_Melee/Pickaxe_VirtualShadowMale1H/FX/NS_Pickaxe_VirtualShadowMale1H_Trail
dependencies {
 usedDataInterfaces [&NiagaraDataInterfaceColorCurve, &NiagaraDataInterfaceCurve, &NiagaraDataInterfaceSocketReader]
 usedDynamicInputs []
 usedModules [
  &self:SystemSpawnScript.SystemScriptSource.SystemScriptGraph.NiagaraNodeAssignment_0.SetVariables_#1
  &SystemState
 ]
}
emitters {
 FN_AnimTrail {
  emitterSpawnScript [&FN_RibbonIDOffsetUtility()]
  emitterUpdateScript [
   &EmitterState()
   &FN_SpawnAnimTrail()
  ]
  particleSpawnScript [
   &FN_InitializeAnimTrailBoneIndecies(First Bone or Socket Index=1, Second Bone or Socket Index=0)
   &FN_InitializeAnimTrailBoneIndecies(AnimTrail Selection=1, Second Bone or Socket Index=0) {module=FN_InitializeAnimTrailBoneIndecies001}
   &FN_InitializeAnimTrail(Color=<rgba 0 0 0 1>, Lifetime=0.22)
   &AddVelocity(Velocity=<0 400 0>) {enabled=false}
  ]
  particleUpdateScript [
   &SolveForcesAndVelocity(Clamp Velocity=false, Limit Acceleration=false)
   &FN_UpdateAnimTrail() {module=FN_UpdateAnimTrail001}
   &ScaleColor(
    Scale Alpha=&Multiply_Float(
     A=&FloatFromCurve(FloatCurve=DataInterface() {propertyValues {Curve {keys [{arriveTangent=-14.0869, interpMode=RCIM_Cubic, leaveTangent=-14.0869, tangentMode=RCTM_User, time=0.3, value=1}, <at 0.5 0 RCIM_Cubic>]}}})
     B=1
    )
    ScaleA=true
    ScaleRGB=true
   )
  ]
  rendererClasses [&NiagaraRibbonRendererProperties]
  renderers [
   {rendererClass=&NiagaraRibbonRendererProperties}
   {rendererClass=&NiagaraRibbonRendererProperties, rendererIndex=1}
   {rendererClass=&NiagaraRibbonRendererProperties, rendererIndex=2}
  ]
 }
 SmokeTrail {
  emitterUpdateScript [
   &EmitterState(Life Cycle Mode=::NewEnumerator1, Loop Behavior=::NewEnumerator1, Loop Delay=0.005, Loop Duration=1, UseLoopDelay=true)
   &SpawnRate(SpawnRate=45, Use Spawn Probability=false)
  ]
  particleSpawnScript [
   &InitializeRibbon(
    Color=<rgba 0 0 0 1>
    Lifetime=0.3
    Module.Write Color=true
    Module.Write Lifetime=true
    Module.Write Mass=false
    Module.Write Material Random=true
    Module.Write Position=false
    Module.Write RibbonTwist=false
    Module.Write RibbonWidth=true
    Ribbon Width=27
   )
   &SocketLocation(SamplePosOffset=false, Socket Reader=$System.Socket Reader, Socket Sampling Mode=::NewEnumerator3, VelocitySpeedLimit=false)
  ]
  particleUpdateScript [
   &ParticleState()
   &SolveForcesAndVelocity(Clamp Velocity=false, Limit Acceleration=false)
   &Color(
    Color=&ColorFromCurve(ColorCurve=DataInterface() {propertyValues {
       AlphaCurve {keys [<at 0.7149 2 RCIM_Cubic>,<at 0.9985 0>]}
       BlueCurve {keys [<at 0.0057 0.0189>,<at 0.2591 0>,<at 1 0.0156>]}
       GreenCurve {keys [<at 0.0057 0>,<at 0.2591 0>,<at 1 0.0026>]}
       RedCurve {keys [<at 0.0057 0.0039>,<at 0.2591 0>,<at 1 0>]}
      }})
    Scale Alpha=0.1
   )
   &ScaleColor(
    Scale Alpha=&Multiply_Float(
     A=&FloatFromCurve(
      FloatCurve=DataInterface() {propertyValues {Curve {keys [
          {interpMode=RCIM_Cubic, leaveTangent=-1.9403, leaveTangentWeight=0.4326, tangentMode=RCTM_Break, time=0.4, value=1}
          {arriveTangent=-1.1087, arriveTangentWeight=0.6923, interpMode=RCIM_Cubic, leaveTangent=-1.1087, tangentMode=RCTM_User, time=1, value=0}
         ]}}}
      Scale Curve=&FNTimeOfDayScalar(Night Scale=0.8, Morning Scale=1, Evening Scale=1)
     )
     B=&MakeCustomFloatFromBool(Bool=$User.bIsFrontEnd, True Float=0.5, False Float=1)
    )
    ScaleA=true
    ScaleRGB=true
   ) {module=ScaleColor002}
   &ScaleRibbonWidth(Ribbon Width Scale=&FloatFromCurve(FloatCurve=DataInterface() {propertyValues {Curve {keys [
         {interpMode=RCIM_Cubic, leaveTangentWeight=0.2293, tangentMode=RCTM_User, time=0.1, value=1}
         {arriveTangent=4.8659, interpMode=RCIM_Cubic, leaveTangent=4.8659, tangentMode=RCTM_User, time=0.75, value=2.75}
         {arriveTangent=24.818, interpMode=RCIM_Cubic, leaveTangent=24.818, tangentMode=RCTM_User, time=1, value=6}
        ]}}}))
   &DynamicMaterialParameters(
    Index 0 Param 1=&FloatFromCurve(
     FloatCurve=DataInterface() {propertyValues {Curve {
        defaultValue=0
        keys [
         {arriveTangent=1.5964, interpMode=RCIM_Cubic, leaveTangent=1.5964, tangentMode=RCTM_User, time=0, value=0}
         {arriveTangent=0.7982, interpMode=RCIM_Cubic, leaveTangent=0.7982, tangentMode=RCTM_User, time=1, value=1}
        ]
       }}}
     Scale Curve=0.75
    )
    Index 0 Param 3=-0.8
    Index 0 Param 4=&FloatFromCurve(FloatCurve=DataInterface() {propertyValues {Curve {keys [
         {interpMode=RCIM_Cubic, tangentMode=RCTM_User, time=0, value=1}
         {arriveTangent=-0.6219, interpMode=RCIM_Cubic, leaveTangent=-0.6219, tangentMode=RCTM_User, time=1, value=0}
        ]}}})
    Module.Param0WriteEnabled=true
    Module.Param1WriteEnabled=true
    Module.Param2WriteEnabled=true
    Module.Param3WriteEnabled=true
   )
   &ScaleColor(
    Scale RGB=&MakeVectorFromLinearColorRGB(Color=&ColorFromCurve(ColorCurve=DataInterface() {propertyValues {
        AlphaCurve {keys [<at 0.7149 2 RCIM_Cubic>,<at 0.9985 0>]}
        BlueCurve {keys [<at 0.0057 0.0189>,<at 0.2591 0>,<at 1 0.0156>]}
        GreenCurve {keys [<at 0.0057 0>,<at 0.2591 0>,<at 1 0.0026>]}
        RedCurve {keys [<at 0.0057 0.0039>,<at 0.2591 0>,<at 1 0>]}
       }}))
    ScaleA=true
    ScaleRGB=true
   )
  ]
  rendererClasses [&NiagaraRibbonRendererProperties]
  renderers [{rendererClass=&NiagaraRibbonRendererProperties}]
 }
}
summary {
 systemName=NS_Pickaxe_VirtualShadowMale1H_Trail
 emitters [
  {emitterName=SmokeTrail, rendererClasses [&NiagaraRibbonRendererProperties]}
  {emitterName=FN_AnimTrail, rendererClasses [&NiagaraRibbonRendererProperties]}
 ]
 userVariables [
  User.bIsFrontEnd {type=NiagaraBool}
  User.RibbonIDOffset {type=NiagaraInt32}
  User.SwingSectionIndex {type=NiagaraInt32}
  User.VariantPicker {type=NiagaraFloat, defaultValue=0, description=""}
 ]
}
